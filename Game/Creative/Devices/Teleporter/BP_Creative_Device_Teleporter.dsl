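; Fortnite Creative teleporter device that moves eligible player pawns to a destination in a target teleporter group, or to itself on event.
; Registers with the minigame's teleporter manager by channel, with enabled phase, team, class and tag checks gating travel.
; Overlap or events trigger teleports, supporting two-way links and group teleports of players inside the effect radius.
; Replicated event type and flash counter drive cosmetic FX, while timelines animate rift and energy materials.
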(event EnergyLevelsUpdated(EnergyComponent)
  ;[1 1 1] Here for V2
  (CallFunction|UpdateTeleporterFX))

(event OnAnyPropertyChanged(ToyOptionsComponent)
  (bind _knob_teleportergroup (Variables|UserOptions|Basic|GetKnob_TeleporterGroup))
  ;[1 1 1] If our channel is changed then we need to re-register ourselves
  (Utilities|FlowControl|Sequence
    (:then_0
      (bind _teleportermanagercomponent (CallFunction|GetTeleporterManager))
      (Utilities|IsValid _teleportermanagercomponent
        (:Is%20Valid
          (Teleporter|UnregisterTeleporterfromGroup _teleportermanagercomponent self (Variables|Default|GetLastChannel))
          (Teleporter|RegisterTeleporterwithGroup _teleportermanagercomponent self _knob_teleportergroup)
          (Variables|Default|SetLastChannel _knob_teleportergroup))))
    (:then_1
      (bind _currentminigame (Variables|/Script/FortniteGame.FortMinigameLogicComponent|GetCurrentMinigame (Variables|Default|GetFortMinigameLogic))
        (:then
          (Class|CreativeEnabledComponent|SetEnabledOnPhase (Variables|Default|GetEnabledComponent) (Variables|/Script/FortniteGame.FortMinigame|GetCurrentState _currentminigame) (Variables|Knob|GetKnob_EnabledIndex)))
        (:else)))
    (:then_2
      (CallFunction|UpdateRiftVisibility)
      (CallFunction|Resave)
      (ClassRestrictions|SetClassWithAffects (Variables|Default|GetCreative_UseRestriction_Component) (Variables|Knob|GetKnob_SelectedClass) (== (Variables|Knob|GetKnob_AffectsClass) (Variables|/Game/Creative/Devices/Common/Components/Creative_UseRestriction_Component.Creative_UseRestriction_Component_C|GetAffectsAllButSelectedValue (Variables|Default|GetCreative_UseRestriction_Component))) true)
      (CallFunction|UpdateEffectRadiusScale))
    (:then_3
      (Default|ResetLink))
    (:then_4
      (CallFunction|InitializeFX))))

(event OnPropertiesLoaded(ToyOptionsComponent)
  (bind _knob_teleportergroup (Variables|UserOptions|Basic|GetKnob_TeleporterGroup))
  ;[1 1 1] If our channel is changed then we need to re-register ourselves
  (Utilities|FlowControl|Sequence
    (:then_0
      (bind _teleportermanagercomponent (CallFunction|GetTeleporterManager))
      (Utilities|IsValid _teleportermanagercomponent
        (:Is%20Valid
          (Teleporter|UnregisterTeleporterfromGroup _teleportermanagercomponent self (Variables|Default|GetLastChannel))
          (Teleporter|RegisterTeleporterwithGroup _teleportermanagercomponent self _knob_teleportergroup)
          (Variables|Default|SetLastChannel _knob_teleportergroup))))
    (:then_1
      (bind _currentminigame (Variables|/Script/FortniteGame.FortMinigameLogicComponent|GetCurrentMinigame (Variables|Default|GetFortMinigameLogic))
        (:then
          (Class|CreativeEnabledComponent|SetEnabledOnPhase (Variables|Default|GetEnabledComponent) (Variables|/Script/FortniteGame.FortMinigame|GetCurrentState _currentminigame) (Variables|Knob|GetKnob_EnabledIndex)))
        (:else)))
    (:then_2
      (CallFunction|UpdateRiftVisibility)
      (CallFunction|Resave)
      (ClassRestrictions|SetClassWithAffects (Variables|Default|GetCreative_UseRestriction_Component) (Variables|Knob|GetKnob_SelectedClass) (== (Variables|Knob|GetKnob_AffectsClass) (Variables|/Game/Creative/Devices/Common/Components/Creative_UseRestriction_Component.Creative_UseRestriction_Component_C|GetAffectsAllButSelectedValue (Variables|Default|GetCreative_UseRestriction_Component))) true)
      (CallFunction|UpdateEffectRadiusScale))
    (:then_3
      (Default|ResetLink))
    (:then_4
      (CallFunction|InitializeFX))))

(event OnAddedtoMinigame(FortMinigameLogic)
  (params (Minigame :type Fort%20Minigame%20Object%20Reference))
  (bind _fortminigamelogic (Variables|Default|GetFortMinigameLogic))
  (bind _currentminigame (Variables|/Script/FortniteGame.FortMinigameLogicComponent|GetCurrentMinigame _fortminigamelogic))
  (bind _knob_teleportergroup (Variables|UserOptions|Basic|GetKnob_TeleporterGroup))
  ;[1 1 1] Register ourself with the current minigame
  (Utilities|FlowControl|Sequence
    (:then_0
      (bind _teleportermanagercomponent (CallFunction|GetTeleporterManager))
      (Utilities|IsValid _teleportermanagercomponent
        (:Is%20Valid
          (Teleporter|RegisterTeleporterwithGroup _teleportermanagercomponent self _knob_teleportergroup)
          (Variables|Default|SetLastChannel _knob_teleportergroup))))
    (:then_1
      (Utilities|IsValid _currentminigame
        (:Is%20Valid
          (Class|CreativeEnabledComponent|SetEnabledOnPhase (Variables|Default|GetEnabledComponent) (Variables|/Script/FortniteGame.FortMinigame|GetCurrentState _currentminigame) (Variables|Knob|GetKnob_EnabledIndex))
          (CallFunction|UpdateRiftVisibility)
          (Initialization|InitializeComponent (Variables|Default|GetCreative_UseRestriction_Component) (Variables|Default|GetFortMinigameLogic) (Variables|Default|GetEnabledComponent) 0 996 -1 0 1 :!class /Game/Creative/Devices/Common/Components/Creative_UseRestriction_Component.Creative_UseRestriction_Component_C)
          (ClassRestrictions|SetClassWithAffects (Variables|Default|GetCreative_UseRestriction_Component) (Variables|Knob|GetKnob_SelectedClass) (== (Variables|Knob|GetKnob_AffectsClass) (Variables|/Game/Creative/Devices/Common/Components/Creative_UseRestriction_Component.Creative_UseRestriction_Component_C|GetAffectsAllButSelectedValue (Variables|Default|GetCreative_UseRestriction_Component))) true))))
    (:then_2
      (Default|ResetLink))))

(event OnComponentBeginOverlap(OverlapCapsule)
  (params (OverlappedComponent :type Primitive%20Component%20Object%20Reference) (OtherActor :type Actor%20Object%20Reference) (OtherComp :type Primitive%20Component%20Object%20Reference) (OtherBodyIndex :type Integer) (bFromSweep :type Boolean) (SweepResult :type Hit%20Result%20Structure%20%28by%20ref%29))
  ;[1 1 1] Teleporter player if it's possible
  (CallFunction|TrytoTeleportActorbyOverlap OtherActor)
  (Networking|SwitchHasAuthority
    (:Authority
      (CallFunction|AddorRemoveBlockingPawns OtherActor true))))

(event EventOnWorldReady
  ;[1 1 1] OnWorldReady setup visibility component so it's aware of our Effect Radius Mesh
  (Visibility|SetVisibilityComponents (Variables|Default|GetCreative_VisibleInGame_Component) (Utilities|Array|MakeArray (Variables|Default|GetEffectRadius)) true))

(event OnComponentBeginOverlap(EffectRadius)
  (params (OverlappedComponent :type Primitive%20Component%20Object%20Reference) (OtherActor :type Actor%20Object%20Reference) (OtherComp :type Primitive%20Component%20Object%20Reference) (OtherBodyIndex :type Integer) (bFromSweep :type Boolean) (SweepResult :type Hit%20Result%20Structure%20%28by%20ref%29))
  ;[1 1 1] Teleporter player if it's possible
  (CallFunction|OnEffectRadiusOverlappedActor OtherActor true))

(event OnComponentEndOverlap(EffectRadius)
  (params (OverlappedComponent :type Primitive%20Component%20Object%20Reference) (OtherActor :type Actor%20Object%20Reference) (OtherComp :type Primitive%20Component%20Object%20Reference) (OtherBodyIndex :type Integer))
  ;[1 1 1] Teleporter player if it's possible
  (CallFunction|OnEffectRadiusOverlappedActor OtherActor))

(event OnComponentEndOverlap(OverlapCapsule)
  (params (OverlappedComponent :type Primitive%20Component%20Object%20Reference) (OtherActor :type Actor%20Object%20Reference) (OtherComp :type Primitive%20Component%20Object%20Reference) (OtherBodyIndex :type Integer))
  ;[1 1 1] Teleporter player if it's possible
  (Networking|SwitchHasAuthority
    (:Authority
      (CallFunction|AddorRemoveBlockingPawns OtherActor))))

(event Custom|Enable%20Event
  (params (Instigator :type Controller%20Object%20Reference))
  ;[1 1 1] Enabled Component Event Handlers
  (Class|CreativeEnabledComponent|SetEnabled (Variables|Default|GetEnabledComponent) true))

(event Custom|Disable%20Event
  (params (Instigator :type Controller%20Object%20Reference))
  ;[1 1 1] Enabled Component Event Handlers
  (Class|CreativeEnabledComponent|SetEnabled (Variables|Default|GetEnabledComponent) false))

(event MinigameStartedDelegate(FortMinigameLogic)
  ;[1 1 1] Turn off hover sphere collision so it doesn't interfere with other collision queries in game
  (Collision|SetCollisionEnabled (Variables|Default|GetHoverOverSphere) "NoCollision")
  (CallFunction|RegisterPlayersinArea))

(event MinigameEndedDelegate(FortMinigameLogic)
  ;[1 1 1] Turn off hover sphere collision so it doesn't interfere with other collision queries in game
  (Collision|SetCollisionEnabled (Variables|Default|GetHoverOverSphere) "QueryOnly")
  (CallFunction|UnregisterAllPlayers))

(event Custom|Activate%20Event
  (params (Instigator :type Controller%20Object%20Reference))
  ;[1 1 1] ActivateWhenReceivingFrom triggers the same flow as if the player had overlapped with the teleport
  (Networking|SwitchHasAuthority
    (:Authority
      (if (Class|CreativeEnabledComponent|IsEnabled (Variables|Default|GetEnabledComponent))
        (Utilities|IsValid Instigator
          (:Is%20Valid
            (bind _asfort_player_controller_athena (Utilities|Casting|CastToFortPlayerControllerAthena Instigator))
            (CallFunction|TrytoTeleportActorbyOverlap (Game|Player|GetPlayerPawn _asfort_player_controller_athena))))))))

(event Custom|TeleportWhenReceiving
  (params (Instigator :type Controller%20Object%20Reference))
  ;[1 1 1] If a pawn sends a teleport to trigger to this teleporter, teleport them to us
  (if (Class|CreativeEnabledComponent|IsEnabled (Variables|Default|GetEnabledComponent))
    (bind _binprogress (CallFunction|IsMinigameinProgress))
    (if _binprogress
      (Utilities|IsValid Instigator
        (:Is%20Valid
          (bind _asfort_player_controller (Utilities|Casting|CastToFortPlayerController Instigator))
          (bind _bcanteleportactor (CallFunction|CanTeleportActor (Game|Player|GetPlayerPawn _asfort_player_controller)))
          (if _bcanteleportactor
            (bind _teleportermanagercomponent (CallFunction|GetTeleporterManager))
            (Utilities|IsValid _teleportermanagercomponent
              (:Is%20Valid
                (CallFunction|TeleportFromEvent (Game|Player|GetPlayerPawn _asfort_player_controller))))))))))

(event Custom|UpdateRiftVisibility
  (bind _enabledcomponent (Variables|Default|GetEnabledComponent))
  (bind _enabled (Class|CreativeEnabledComponent|IsEnabled _enabledcomponent))
  (bind _knob_bteleportervisible (Variables|Knob|GetKnob_bTeleporterVisible))
  (bind _returnvalue (and _knob_bteleportervisible _enabled))
  (bind _portal_inworld_loop_cue (Variables|Default|GetPortal_InWorld_Loop_Cue))
  ;[1 1 1] If the game phase allows visibility and our teleporter is set to be visible or we're in creation mode, show the rift mesh and VFX
  (Networking|CanExecuteCosmeticEvents
    (:True
      (bind _bincreationmode (CallFunction|IsMinigameinCreationMode))
      (Rendering|SetVisibility (Variables|Debug|GetStaticMeshComponent) (or _returnvalue _bincreationmode) false)
      (Rendering|SetVisibility (Variables|VFX|GetP_Rift_Idle_Loop) (or _returnvalue _bincreationmode) false)
      (Rendering|SetVisibility (Variables|Default|GetNS_Rift_Portal_Closing) (or _returnvalue _bincreationmode) false)
      (if (or _returnvalue _bincreationmode)
        (Audio|Components|Audio|Play _portal_inworld_loop_cue 0.0)
        (else
          (Audio|Components|Audio|Stop _portal_inworld_loop_cue))))))

(event EventEndPlay
  (params (EndPlayReason :type EEndPlayReason%20Enum :default "Destroyed"))
  ;[1 1 1] Handle Clean-up
  (|CallParentFunction EndPlayReason)
  (Networking|SwitchHasAuthority
    (:Authority
      (bind _teleportermanagercomponent (CallFunction|GetTeleporterManager))
      (Utilities|IsValid _teleportermanagercomponent
        (:Is%20Valid
          (Teleporter|UnregisterTeleporterfromGroup _teleportermanagercomponent self (Variables|Default|GetLastChannel)))))))

(event MinigameStateChangedDelegate(FortMinigameLogic)
  (params (Minigame :type Fort%20Minigame%20Object%20Reference) (NewMinigameState :type EFortMinigameState%20Enum :default "PreGame"))
  (Class|CreativeEnabledComponent|SetEnabledOnPhase (Variables|Default|GetEnabledComponent) NewMinigameState (Variables|Knob|GetKnob_EnabledIndex))
  (CallFunction|UpdateRiftVisibility)
  (Default|ResetLink)
  (Variables|Default|SetbShuffledOnce false))

(event OnEnabledStateChanged(EnabledComponent)
  (params (Enabled :type Boolean))
  ;[1 1 1] Handle enable component events, force enable for creation mode
  (Networking|SwitchHasAuthority
    (:Authority
      (if Enabled
        (Variables|VFX|SetTeleporterEventType "Enabled")
        (else
          (Variables|VFX|SetTeleporterEventType "Disabled")))
      (Variables|VFX|SetTeleporterEventFlash (Math|Byte|Byte+Byte (Variables|VFX|GetTeleporterEventFlash) 1)))))

(event Custom|ActivateLink
  (params (Instigator :type Controller%20Object%20Reference))
  (Variables|Default|SetbLinkToTarget true))

(event Custom|DeactivateLink
  (params (Instigator :type Controller%20Object%20Reference))
  (Variables|Default|SetbLinkToTarget false)
  (Default|ResetLink))

(event Custom|ResetLinkEvent
  (params (Instigator :type Controller%20Object%20Reference))
  (Default|ResetLink))

(event Custom|Teleport%20from%20Event
  (params (PlayerPawn :type Fort%20Player%20Pawn%20Object%20Reference))
  (bind _charactermovement (Variables|/Script/Engine.Character|GetCharacterMovement PlayerPawn))
  ;[1 1 1] Handle Teleport From Event
  (switch Utilities|FlowControl|Switch|SwitchonE_TeleporterDevice_TeleportToEventMomentum (Variables|UserOptions|Advanced|GetMaintainMomentumDuringTeleportToEvent)
    (:NewEnumerator0
      _)
    (:NewEnumerator1
      (Variables|Default|SetCachedPendingVelocity (Pawn|Character|GetPendingLaunchVelocity PlayerPawn))
      (Pawn|Components|CharacterMovement|ClearAccumulatedForces (Variables|/Script/Engine.Character|GetCharacterMovement PlayerPawn))
      (Variables|/Script/Engine.MovementComponent|SetVelocity (* (Math|Float|SelectFloat (Math|Vector|VectorLength (Variables|Default|GetCachedPendingVelocity)) (Math|Vector|VectorLength (Variables|/Script/Engine.MovementComponent|GetVelocity (Variables|/Script/Engine.Character|GetCharacterMovement PlayerPawn))) (Math|Vector|NotEqualExactly(Vector) (Variables|Default|GetCachedPendingVelocity))) (Math|Vector|Normalize (Transformation|GetActorForwardVector))) (Variables|/Script/Engine.Character|GetCharacterMovement PlayerPawn)))
    (:NewEnumerator2
      (Character|StopJumping PlayerPawn)
      (FortMovement|CancelJumpAbility PlayerPawn false)
      (Components|Movement|StopMovementImmediately _charactermovement)
      (Pawn|Components|CharacterMovement|ClearAccumulatedForces _charactermovement)
      (CallFunction|TeleportPlayerPawnTo PlayerPawn self true))))

(event Custom|PlayTeleportExitedFX
  (params (Location :type Vector))
  (bind _exitriftps (Variables|VFX|GetExitRiftPS))
  (bind _enteredriftsound (Variables|Sound|GetEnteredRiftSound))
  ;[1 1 1] On Teleporter Exited play a SFX
  (Networking|CanExecuteCosmeticEvents
    (:True
      (Utilities|FlowControl|Sequence
        (:then_0
          (if (Variables|Knob|GetKnob_bVFXEnabled)
            (Utilities|IsValid _exitriftps
              (:Is%20Valid
                (Effects|SpawnEmitteratLocation _exitriftps Location)))))
        (:then_1
          (if (Variables|Knob|GetKnob_bSFXEnabled)
            ;;[1 1 1] Delay to allow for teleported player to hear exit sound
            (Utilities|FlowControl|Delay 0.1)
            (Utilities|IsValid _enteredriftsound
              (:Is%20Valid
                (Audio|PlaySoundatLocation _enteredriftsound Location "0, 0, 0" 1.0 1.0 0.0 0 "/Game/Athena/Sounds/Events/RiftBeacon/RiftBeacon_SingleVoice_Conc.RiftBeacon_SingleVoice_Conc")))))))))

(event Custom|PlayTeleportEnteredFX
  (params (Location :type Vector))
  (bind _enteredriftsound (Variables|Sound|GetEnteredRiftSound))
  ;[1 1 1] On teleporter entered, play SFX and VFX
  (Networking|CanExecuteCosmeticEvents
    (:True
      (Utilities|FlowControl|Sequence
        (:then_0
          (if (Variables|Knob|GetKnob_bVFXEnabled)
            (Effects|SpawnEmitteratLocation (Variables|VFX|GetEnterRiftPS) Location)))
        (:then_1
          (if (Variables|Knob|GetKnob_bSFXEnabled)
            (Utilities|IsValid _enteredriftsound
              (:Is%20Valid
                (Audio|PlaySoundatLocation _enteredriftsound Location "0, 0, 0" 1.0 1.0 0.0 0 "/Game/Athena/Sounds/Events/RiftBeacon/RiftBeacon_SingleVoice_Conc.RiftBeacon_SingleVoice_Conc")))))))))

(event Custom|UpdateTeleporterFX
  (bind _energycostattribute (Variables|Energy|GetEnergyCostAttribute))
  (bind _self self)
  (bind _returnvalue (Ability|GetAbilitySystemComponent _self))
  (bind (_bsuccessfullyfoundattribute _returnvalue_2) (Ability|Attribute|GetFloatAttributefromAbilitySystemComponent _returnvalue _energycostattribute))
  (bind _energycapacityattribute (Variables|Energy|GetEnergyCapacityAttribute))
  (bind (_bsuccessfullyfoundattribute_2 _returnvalue_3) (Ability|Attribute|GetFloatAttributefromAbilitySystemComponent _returnvalue _energycapacityattribute))
  (bind _currentenergyattribute (Variables|Energy|GetCurrentEnergyAttribute))
  (bind (_bsuccessfullyfoundattribute_3 _returnvalue_4) (Ability|Attribute|GetFloatAttributefromAbilitySystemComponent _returnvalue _currentenergyattribute))
  (bind _currentenergycapacity (Variables|Transient|GetCurrentEnergyCapacity))
  ;[1 1 1] Teleporter UX FX for V2
  (Networking|CanExecuteCosmeticEvents
    (:True
      (if (Math|Boolean|ANDBoolean _bsuccessfullyfoundattribute _bsuccessfullyfoundattribute_3 _bsuccessfullyfoundattribute_2)
        (Variables|Transient|SetCurrentEnergyCost _returnvalue_2)
        (Variables|Transient|SetCurrentEnergyCapacity _returnvalue_3)
        (Variables|Transient|SetPreviousEnergyLevel (Variables|Transient|GetDesiredEnergyLevel))
        (Variables|Transient|SetDesiredEnergyLevel _returnvalue_4)
        (bind (_direction _linear) (Timeline|TPEffectsTimeLine PlayFromStart)
          (:Update
            (if (> (Variables|Transient|GetCurrentEnergyCapacity) 0.0)
              (bind _output_get (Variables|Transient|SetNewEnergyLevel (Math|Float|Lerp (Variables|Transient|GetPreviousEnergyLevel) (Variables|Transient|GetDesiredEnergyLevel) _linear)))
              (Rendering|Material|SetScalarParameterValueonMaterials (Variables|Debug|GetStaticMeshComponent) "Emissive" (/ _output_get _currentenergycapacity))
              (bind _output_get_2 (Variables|Transient|SetNewTeleporterEnergyColor (Math|Color|LinearColor*Float (Math|Color|Lerp(LinearColor) (Variables|Default|GetTeleporterLowEnergyColor) (Variables|Default|GetTeleporterMaxEnergyColor) (/ _output_get _currentenergycapacity)) (/ _output_get _currentenergycapacity))))
              (Rendering|Material|SetVectorParameterValueonMaterials (Variables|Debug|GetStaticMeshComponent) "CubemapTint" (Math|Conversions|ToVector(LinearColor) _output_get_2))
              (if (>= (Variables|Transient|GetCurrentEnergyCost) (Variables|Transient|GetCurrentEnergyCapacity))
                (Variables|VFX|SetFragmentMaskVector (select (>= (Variables|Transient|GetNewEnergyLevel) (Variables|Transient|GetCurrentEnergyCost)) "1.000000,1.000000,1.000000" "0, 0, 0"))
                (else
                  (Variables|VFX|SetFragmentMaskVector (select (> (Variables|Transient|GetNewEnergyLevel) 0.0) (select (> (Variables|Transient|GetNewEnergyLevel) (Variables|Transient|GetCurrentEnergyCost)) (Math|Vector|MakeVector (select (>= (/ (- (Variables|Transient|GetNewEnergyLevel) (Variables|Transient|GetCurrentEnergyCost)) (- (Variables|Transient|GetCurrentEnergyCapacity) (Variables|Transient|GetCurrentEnergyCost))) 0.0) (Math|Float|Clamp(Float) (/ (/ (- (Variables|Transient|GetNewEnergyLevel) (Variables|Transient|GetCurrentEnergyCost)) (- (Variables|Transient|GetCurrentEnergyCapacity) (Variables|Transient|GetCurrentEnergyCost))) 0.16)) 0.0) (select (>= (/ (- (Variables|Transient|GetNewEnergyLevel) (Variables|Transient|GetCurrentEnergyCost)) (- (Variables|Transient|GetCurrentEnergyCapacity) (Variables|Transient|GetCurrentEnergyCost))) 0.33) (Math|Float|Clamp(Float) (/ (- (/ (- (Variables|Transient|GetNewEnergyLevel) (Variables|Transient|GetCurrentEnergyCost)) (- (Variables|Transient|GetCurrentEnergyCapacity) (Variables|Transient|GetCurrentEnergyCost))) 0.33) 0.16)) 0.0) (select (>= (/ (- (Variables|Transient|GetNewEnergyLevel) (Variables|Transient|GetCurrentEnergyCost)) (- (Variables|Transient|GetCurrentEnergyCapacity) (Variables|Transient|GetCurrentEnergyCost))) 0.66) (Math|Float|Clamp(Float) (/ (- (/ (- (Variables|Transient|GetNewEnergyLevel) (Variables|Transient|GetCurrentEnergyCost)) (- (Variables|Transient|GetCurrentEnergyCapacity) (Variables|Transient|GetCurrentEnergyCost))) 0.66) 0.16)) 0.0)) "0, 0, 0") "1.000000,1.000000,1.000000"))))
              (Rendering|Material|SetVectorParameterValueonMaterials (Variables|Debug|GetStaticMeshComponent) "FragmentMask" (Variables|VFX|GetFragmentMaskVector))
              (Rendering|Material|SetVectorParameterValue :self (Variables|VFX|GetSharpFlareMID) :ParameterName "DarkColor" :!class /Script/Engine.MaterialInstanceDynamic)
              (Rendering|Material|SetVectorParameterValue (Variables|VFX|GetSharpFlareMID) "TintColor" (Variables|Transient|GetNewTeleporterEnergyColor) :!class /Script/Engine.MaterialInstanceDynamic)
              (Rendering|Material|SetScalarParameterValue (Variables|VFX|GetCoronaMID) "Brightness" (/ _output_get _currentenergycapacity) :!class /Script/Engine.MaterialInstanceDynamic)
              (Rendering|Material|SetScalarParameterValue (Variables|VFX|GetRoundFlareMID) "CenterBrightness" (* (Variables|VFX|GetRiftCenterBrightnessScalar) (/ _output_get _currentenergycapacity)) :!class /Script/Engine.MaterialInstanceDynamic)
              (Rendering|Material|SetScalarParameterValue (Variables|VFX|GetRoundFlareMID) "EdgeBrightness" (* (Variables|VFX|GetRiftEdgeBrightnessScalar) (/ _output_get _currentenergycapacity)) :!class /Script/Engine.MaterialInstanceDynamic))))))))

(event Custom|PortalEnabledFX
  (bind _staticmeshcomponent (Variables|Debug|GetStaticMeshComponent))
  ;[1 1 1] Open Rift
  (Networking|CanExecuteCosmeticEvents
    (:True
      (Effects|Components|ParticleSystem|SetTemplate (Variables|VFX|GetP_Rift_Idle_Loop) "/Game/Creative/Devices/Teleporter/P_RefinedRift_Teleport_Idle.P_RefinedRift_Teleport_Idle")
      (CallFunction|InitializeFX)
      (Rendering|SetVisibility _staticmeshcomponent true false)
      (bind (_direction _linear) (Timeline|TPEnabledFXTimeline PlayFromStart)
        (:Update
          (Rendering|Material|SetCustomPrimitiveDataFloat _staticmeshcomponent 0 _linear))
        (:Finished
          (CallFunction|UpdateRiftVisibility))))))

(event Custom|PortalDisabledFX
  ;[1 1 1] Close Rift
  (Networking|CanExecuteCosmeticEvents
    (:True
      (Effects|Components|ParticleSystem|SetTemplate (Variables|VFX|GetP_Rift_Idle_Loop) "/Game/Creative/Items/Gameplay/Teleporter/P_Rift_Portal_Idle.P_Rift_Portal_Idle")
      (bind (_direction _linear) (Timeline|TPDisabledFXTimeline PlayFromStart)
        (:Update
          (Rendering|Material|SetCustomPrimitiveDataFloat (Variables|Debug|GetStaticMeshComponent) 0 _linear))
        (:Finished
          (CallFunction|UpdateRiftVisibility))))))

(fn AddOrRemoveBlockingPawns
  (params (FortPlayerPawn :type Actor%20Object%20Reference%20%28by%20ref%29) (AddPawn :type Boolean))
  (bind _asfort_player_pawn (Utilities|Casting|CastToFortPlayerPawn FortPlayerPawn))
  (if AddPawn
    (Utilities|Set|Add (Variables|Default|GetPawnsOnTeleporter) _asfort_player_pawn)
    (else
      (Utilities|Set|Remove (Variables|Default|GetPawnsOnTeleporter) _asfort_player_pawn))))

(fn BindOrUnbindToPlayerRegisterEvents
  (params (FortPlayerPawnAthena :type Fort%20Player%20Pawn%20Athena%20Object%20Reference%20%28by%20ref%29) (Bind :type Boolean))
  (bind _outputdelegate (EventDispatchers|CreateEvent :!event OnPlayerDied))
  (Utilities|IsValid FortPlayerPawnAthena
    (:Is%20Valid
      (if Bind
        (EventDispatchers|BindEventtoOnFortPlayerDied FortPlayerPawnAthena _outputdelegate)
        (else
          (EventDispatchers|UnbindEventfromOnFortPlayerDied FortPlayerPawnAthena _outputdelegate))))))

(fn BlueprintCanInteract
  (params (InteractingPawn :type Fort%20Pawn%20Object%20Reference) (InteractionBeingAttempted :type EInteractionBeingAttempted%20Enum :default "FirstInteraction") (InteractionType :type TInteractionType%20Enum :default "IT_NoInteraction"))
  (bind _returnvalue (Pawn|GetController InteractingPawn))
  (bind _fortminigamelogic (Variables|Default|GetFortMinigameLogic))
  (bind _currentminigame (Variables|/Script/FortniteGame.FortMinigameLogicComponent|GetCurrentMinigame _fortminigamelogic))
  (Utilities|FlowControl|Sequence
    (:then_0
      (Utilities|IsValid InteractingPawn
        (:Is%20Valid
          (Utilities|IsValid _returnvalue
            (:Is%20Valid
              (bind _asfort_player_controller_athena (Utilities|Casting|CastToFortPlayerControllerAthena _returnvalue))
              (if (.ReturnValue (Creative|CanPlayerCreateinCurrentVolume _asfort_player_controller_athena :!class /Script/FortniteGame.FortPlayerControllerAthena))
                (Utilities|IsValid _currentminigame
                  (:Is%20Valid
                    (LocalVariables|SetbInteractSuccess (not (Minigame|HasStarted _currentminigame :!class /Script/FortniteGame.FortMinigame))))
                  (:Is%20Not%20Valid
                    (LocalVariables|SetbInteractSuccess true)))))))))
    (:then_1
      (return (LocalVariables|GetbInteractSuccess)))))

(fn CanTeleportActor
  (params (PawnToTeleport :type Fort%20Pawn%20Object%20Reference) (bIgnoreTags :type Boolean))
  (Utilities|IsValid PawnToTeleport
    (:Is%20Valid
      (if (Class|CreativeEnabledComponent|IsEnabled (Variables|Default|GetEnabledComponent))
        ;[1 1 1] Check is player is in creative ghost/phase mode
        (bind _asfort_player_pawn_athena (Utilities|Casting|CastToFortPlayerPawnAthena PawnToTeleport)
          (:then
            (if (not (Variables|/Script/FortniteGame.FortPlayerPawnAthena|GetbIsCreativeGhostModeActivated _asfort_player_pawn_athena))
              ;[1 1 1] Check that the player is on the right team to use the teleporter
              (if (or (Knobs|Team|IsTeamAll (Variables|Knob|GetKnob_AllowedTeamIdx)) (== (Knobs|Team|GetTeamValue (Variables|Knob|GetKnob_AllowedTeamIdx)) (Math|Conversions|ToInteger(Byte) (Math|Byte|Byte-Byte (Team|GetActorTeam PawnToTeleport) (Utilities|Enum|LiteralenumEFortTeam "Monster")))))
                ;[1 1 1] Check that the player is the right class to use the teleporter
                (bind _valid_controller (DEPRECATED|ActorHasValidController (Variables|Default|GetCreative_UseRestriction_Component) PawnToTeleport true true))
                (if _valid_controller
                  ;[1 1 1] Don't allow pawns to teleport if they're in vehicles until V2, unexpected behaviour otherwise
                  (Utilities|IsinVehicle PawnToTeleport
                    (:true
                      _)
                    (:false
                      (if (or (Variables|GetbIgnoreTags) (not (GameplayTags|HasAnyMatchingGameplayTags PawnToTeleport (Variables|Default|GetTeleporterBlockingTags))))
                        (return true)
                        (else
                          (LocalVariables|SetDenyTeleportReason "Cooldown teleport is active")))))
                  (if (CallFunction|IsPortalDebuggingEnabled)
                    (Development|PrintString (LocalVariables|GetDenyTeleportReason) true true "(R=0.456733,G=0.960172,B=1.000000,A=1.000000)" 5.0))
                  (return false)
                  (else
                    (LocalVariables|SetDenyTeleportReason "Actor has not a valid controller")))
                (else
                  (LocalVariables|SetDenyTeleportReason "Actor team does not correspond to knob value")))))
          (:CastFailed
            _))
        (else
          (LocalVariables|SetDenyTeleportReason "EnabledComponent is disabled"))))
    (:Is%20Not%20Valid
      (LocalVariables|SetDenyTeleportReason "Pawn is invalid"))))

(fn GetTeleportedPawnPitch
  (params (PawnToTeleport :type Fort%20Player%20Pawn%20Object%20Reference) (SourceTeleporter :type BP%20Creative%20Device%20Teleporter%20Object%20Reference))
  (return (Utilities|Select :Index (Variables|Knob|GetKnob_bFacePlayerInDirectionOfTeleporter) :Option%200 (.pitch (Pawn|GetControlRotation (Pawn|GetController PawnToTeleport) :!class /Script/Engine.Controller)) :Option%201 (+ (.pitch (Math|Rotator|Delta(Rotator) (Pawn|GetControlRotation (Pawn|GetController PawnToTeleport) :!class /Script/Engine.Controller) (Transformation|GetActorRotation SourceTeleporter))) (.pitch (Transformation|GetActorRotation))) :Option%202 (.pitch (Transformation|GetActorRotation)))))

(fn GetTeleportedPawnYaw
  (params (PawnToTeleport :type Fort%20Player%20Pawn%20Object%20Reference) (SourceTeleporter :type BP%20Creative%20Device%20Teleporter%20Object%20Reference))
  (return (Utilities|Select :Index (Variables|Knob|GetKnob_bFacePlayerInDirectionOfTeleporter) :Option%200 (.yaw (Transformation|GetActorRotation PawnToTeleport)) :Option%201 (+ (.yaw (Math|Rotator|Delta(Rotator) (Transformation|GetActorRotation PawnToTeleport) (Transformation|GetActorRotation SourceTeleporter))) (.yaw (Transformation|GetActorRotation))) :Option%202 (.yaw (Transformation|GetActorRotation)))))

(fn GetTeleporterManager
  (bind _self self)
  (bind _returnvalue (Minigame|GetMinigame _self :!class /Script/FortniteGame.MinigameBlueprintLibrary))
  (Utilities|IsValid _returnvalue
    (:Is%20Valid
      (return (Variables|/Script/FortniteGame.FortMinigame|GetTeleporterManagerComponent _returnvalue)))))

(fn InitializeFX
  (Rendering|Material|SetCustomPrimitiveDataFloat (Variables|Debug|GetStaticMeshComponent) 0 (select (Class|CreativeEnabledComponent|IsEnabled (Variables|Default|GetEnabledComponent)) 1.0 0.0)))

(fn isMinigameInCreationMode
  (Utilities|FlowControl|Sequence
    (:then_0
      (bind _currentminigame (Variables|/Script/FortniteGame.FortMinigameLogicComponent|GetCurrentMinigame (Variables|Default|GetFortMinigameLogic))
        (:then
          (LocalVariables|SetbSuccess (Utilities|Enum|Equal(Enum) (Variables|/Script/FortniteGame.FortMinigame|GetCurrentState _currentminigame) "PreGame")))
        (:else
          (LocalVariables|SetbSuccess (Editor|InEditor)))))
    (:then_1
      (return (LocalVariables|GetbSuccess)))))

(fn isMinigameInProgress
  (bind _fortminigamelogic (Variables|Default|GetFortMinigameLogic))
  (bind _currentminigame (Variables|/Script/FortniteGame.FortMinigameLogicComponent|GetCurrentMinigame _fortminigamelogic))
  (Utilities|FlowControl|Sequence
    (:then_0
      (Utilities|IsValid _currentminigame
        (:Is%20Valid
          (LocalVariables|SetbLocalInProgress (Utilities|Enum|Equal(Enum) (Variables|/Script/FortniteGame.FortMinigame|GetCurrentState _currentminigame) "InProgress")))))
    (:then_1
      (return (LocalVariables|GetbLocalInProgress)))))

(fn OnEffectRadiusOverlappedActor
  (params (OverlappingActor :type Actor%20Object%20Reference%20%28by%20ref%29) (Enter :type Boolean))
  (Networking|SwitchHasAuthority
    (:Authority
      (if (> (Variables|Knob|GetKnob_EffectRadiusSize) 0.0)
        (bind _asfort_player_pawn_athena (Utilities|Casting|CastToFortPlayerPawnAthena OverlappingActor))
        (if (Variables|GetEnter)
          (CallFunction|RegisterPlayer _asfort_player_pawn_athena)
          (else
            (CallFunction|UnregisterPlayer _asfort_player_pawn_athena)))))))

(fn OnTeleportSucceded
  (params (TargetTeleporter :type BP%20Creative%20Device%20Teleporter%20Object%20Reference%20%28by%20ref%29) (TriggerInstigator :type Actor%20Object%20Reference%20%28by%20ref%29))
  (Networking|SwitchHasAuthority
    (:Authority
      (if (and (> (Variables|Knob|GetKnob_EffectRadiusSize) 0.0) (not (Variables|Default|GetIgnoreEffectRadiusSettings)))
        ;[1 1 1] This function will be called multiple times recursively. This will avoid that the function does something while this happens
        (if (not (Variables|Default|GetbIsTeleportingInRadiusPawns))
          (Variables|Default|SetbIsTeleportingInRadiusPawns true)
          (Utilities|IsValid TriggerInstigator
            (:Is%20Valid
              (bind (_arrayindex _arrayelement) (Utilities|Array|ReverseforEachLoop (Variables|Default|GetRegisteredPlayers))
                (:LoopBody
                  (if (Utilities|NotEqual(Object) TriggerInstigator _arrayelement)
                    (CallFunction|TrytoTeleportActorbyOverlap _arrayelement (Variables|GetTargetTeleporter))))
                (:Completed
                  _)))
            (:Is%20Not%20Valid
              _))
          (Variables|Default|SetbIsTeleportingInRadiusPawns false))))))

(fn OnPlayerDied
  (params (FortPlayerPawnAthena :type Fort%20Player%20Pawn%20Athena%20Object%20Reference))
  (CallFunction|UnregisterPlayer FortPlayerPawnAthena))

(fn OnReceivedPawn
  (params (PawnReceived :type Fort%20Player%20Pawn%20Object%20Reference))
  (bind _returnvalue (Pawn|GetController PawnReceived))
  (Utilities|IsValid PawnReceived
    (:Is%20Valid
      (LocalVariables|SetPlayerPawn PawnReceived)
      (Utilities|IsValid _returnvalue
        (:Is%20Valid
          (Utilities|Time|ClearandInvalidateTimerbyHandle (Variables|VFX|GetTeleportEventResetTimer))
          (Variables|VFX|SetTeleporterEventType "Exited")
          (Variables|VFX|SetTeleporterEventFlash (Math|Byte|Byte+Byte (Variables|VFX|GetTeleporterEventFlash) 1))
          (bind _returnvalue_2 (Utilities|Time|SetTimerbyEvent (EventDispatchers|CreateEvent :!event ResetTeleportEventState) 0.1))
          (Variables|VFX|SetTeleportEventResetTimer _returnvalue_2)
          (GameplayEvent|BroadcastEvent (Variables|UserOptions|Events|GetOn%20Teleported) 0 (Utilities|Struct|MakeGameplayEventPlayerTeleported _returnvalue)))))))

(fn OnRep_TeleporterUsedFlash
  (Networking|CanExecuteCosmeticEvents
    (:True
      (switch Utilities|FlowControl|Switch|SwitchonEFortCreativeTeleporterEvent (Variables|VFX|GetTeleporterEventType)
        (:Entered
          (CallFunction|PlayTeleportEnteredFX (Transformation|GetActorLocation)))
        (:Exited
          (CallFunction|PlayTeleportExitedFX (Transformation|GetActorLocation)))
        (:Enabled
          (CallFunction|PortalEnabledFX))
        (:Disabled
          (CallFunction|PortalDisabledFX))))))

(fn RegisterPlayer
  (params (FortPlayerPawn :type Fort%20Player%20Pawn%20Athena%20Object%20Reference%20%28by%20ref%29))
  (bind _fortplayerpawn (Variables|GetFortPlayerPawn))
  (Utilities|IsValid FortPlayerPawn
    (:Is%20Valid
      (if (not (FortPawn|IsDead FortPlayerPawn))
        (bind _returnvalue (Utilities|Array|AddUnique (Variables|Default|GetRegisteredPlayers) FortPlayerPawn))
        (if (not (== _returnvalue (LocalVariables|GetIndex_None)))
          (CallFunction|BindorUnbindtoPlayerRegisterEvents _fortplayerpawn true)
          (if (CallFunction|IsPortalDebuggingEnabled)
            (Development|PrintString (Utilities|String|Append "Registered Player: " (Utilities|GetDisplayName _fortplayerpawn) "In: " (Utilities|GetDisplayName self)) true true "(R=1.000000,G=0.720884,B=0.660858,A=1.000000)" 1e+01)))))))

(fn RegisterPlayersInArea
  (Networking|SwitchHasAuthority
    (:Authority
      ;[1 1 1] Make sure list is empty
      (CallFunction|UnregisterAllPlayers)
      (bind _output_get (LocalVariables|SetOverlappingActors (Collision|GetOverlappingActors (Variables|Default|GetEffectRadius) "/Script/FortniteGame.FortPlayerPawn" :!class /Script/Engine.PrimitiveComponent)))
      (for _array_element _output_get
        (CallFunction|OnEffectRadiusOverlappedActor _array_element true)))))

(fn ResetLink
  (bind _destinationteleporter (Variables|Default|GetDestinationTeleporter)
    (:then
      (if (Variables|Default|GetbIsLinked _destinationteleporter)
        (if (== (Variables|Default|GetDestinationTeleporter _destinationteleporter) self)
          (CallFunction|SetDestinationTeleporter _destinationteleporter 0)
          (CallFunction|SetIsLinked _destinationteleporter false)
          (else
            _))
        (CallFunction|SetDestinationTeleporter 0)
        (bind _binprogress (CallFunction|IsMinigameinProgress))
        (if (not _binprogress)
          (Variables|Default|SetbLinkToTarget (Variables|Default|GetbInitialLinkToTarget)))
        (CallFunction|SetIsLinked false)
        (else
          _)))
    (:else
      _)))

(fn ResetTeleportEventState
  ;[1 1 1] To prevent initial clients getting bombarded with enter and exit VFX we set the event type to none
  (Networking|SwitchHasAuthority
    (:Authority
      (switch Utilities|FlowControl|Switch|SwitchonEFortCreativeTeleporterEvent (Variables|VFX|GetTeleporterEventType)
        (:Entered
          _)
        (:Exited
          _))
      (Variables|VFX|SetTeleporterEventType "None"))))

(fn SelectDestinationTeleporter
  (params (PawnToTeleport :type Fort%20Player%20Pawn%20Object%20Reference))
  (bind _fortminigamelogic (Variables|Default|GetFortMinigameLogic))
  (bind _currentminigame (Variables|/Script/FortniteGame.FortMinigameLogicComponent|GetCurrentMinigame _fortminigamelogic))
  (bind _returnvalue (Minigame|GetRuntime _currentminigame))
  (bind _knob_changetarget (Variables|Knob|GetKnob_ChangeTarget))
  (bind _returnvalue_2 (Math|Conversions|ToFloat(Integer) _knob_changetarget))
  (bind _returnvalue_3 (/ _returnvalue _returnvalue_2))
  (bind _returnvalue_4 (Math|Float|Floor _returnvalue_3))
  (bind _destinationteleporter (Variables|Default|GetDestinationTeleporter))
  (if (Variables|Default|GetbLinkToTarget)
    (Utilities|IsValid _destinationteleporter
      (:Is%20Valid
        (if (Variables|Default|GetbLinkToTarget _destinationteleporter)
          (if (Variables|Default|GetbIsLinked _destinationteleporter)
            (if (== (Variables|Default|GetDestinationTeleporter _destinationteleporter) self)
              (return _destinationteleporter)
              (else
                _))
            (else
              _))
          (LocalVariables|SetLocalPawnToTeleport PawnToTeleport)
          (bind _teleportermanagercomponent (CallFunction|GetTeleporterManager))
          (Utilities|IsValid _teleportermanagercomponent
            (:Is%20Valid
              (bind _result (Utilities|Set|ToArray (Teleporter|GetTeleportersinGroup _teleportermanagercomponent (Variables|UserOptions|Basic|GetKnob_TargetTeleporterGroup))))
              (if (> (Utilities|Array|Length _result) 0)
                (if (== (Variables|Knob|GetKnob_ChangeTarget) -2)
                  (if (not (Variables|Default|GetbShuffledOnce))
                    (Utilities|Array|Shuffle _result)
                    (Variables|Default|SetbShuffledOnce true)
                    (else
                      _))
                  (else
                    (if (== (Variables|Knob|GetKnob_ChangeTarget) -1)
                      ;[1 1 1] Pick the first in the list
                      (for _array_element _result
                        (LocalVariables|SetCurrentTeleporterCandidate _array_element)
                        ;[1 1 1] Loop Body
                        (bind _currentteleportercandidate (LocalVariables|GetCurrentTeleporterCandidate)
                          (:then
                            (bind _asbp_creative_device_teleporter (Utilities|Casting|CastToBP_Creative_Device_Teleporter _currentteleportercandidate))
                            (if (and (Class|CreativeEnabledComponent|IsEnabled (Variables|Default|GetEnabledComponent _asbp_creative_device_teleporter)) (Collision|GetActorEnableCollision _asbp_creative_device_teleporter))
                              (if (Utilities|NotEqual(Object) self _asbp_creative_device_teleporter)
                                ;[1 1 1] If we intend to pair with the new destination teleporter, make sure we are not breaking an already existing pair
                                (if (Variables|Default|GetbLinkToTarget)
                                  (if (Variables|Default|GetbIsLinked _asbp_creative_device_teleporter)
                                    _
                                    (else
                                      _))
                                  (else
                                    _))
                                (bind _outyaw (CallFunction|GetTeleportedPawnYaw _asbp_creative_device_teleporter (LocalVariables|GetLocalPawnToTeleport) self))
                                (if (not (Teleporter|IsTeleporterBlocked _asbp_creative_device_teleporter (LocalVariables|GetLocalPawnToTeleport) (Math|Rotator|MakeRotator (.roll (Transformation|GetActorRotation (LocalVariables|GetLocalPawnToTeleport))) (.pitch (Transformation|GetActorRotation (LocalVariables|GetLocalPawnToTeleport))) _outyaw)))
                                  (Variables|Default|SetCurrentTimeSelectedTarget _asbp_creative_device_teleporter)
                                  (return _asbp_creative_device_teleporter)))))
                          (:else)))
                      (return 0)
                      (else
                        ;[1 1 1] According to design all teleporter behaviours should operate on a random ordering of the teleporter list
                        (Utilities|Array|Shuffle _result)
                        (if (== (Variables|Knob|GetKnob_ChangeTarget) 0)
                          ;[1 1 1] On teleporter entered select the next random teleporter
                          (for _index (range (+ (Utilities|Array|Length _result) 1))
                            (LocalVariables|SetCurrentTeleporterCandidate (Utilities|Array|Get(acopy) _result (Math|Integer|%25%28Integer%29 (+ _index (Variables|Default|GetTeleporterUseCounter)) (Utilities|Array|Length _result)))))
                          (return 0)
                          (else
                            ;[1 1 1] On teleporter entered select based on how much time has passed and how often we should be switching
                            (Utilities|IsValid _currentminigame
                              (:Is%20Valid
                                (if (!= _returnvalue_4 (Variables|Default|GetCurrentTimeSelectedTargetIndex))
                                  (Variables|Default|SetCurrentTimeSelectedTargetIndex _returnvalue_4)
                                  (for _index_2 (range (+ (Utilities|Array|Length _result) 1))
                                    (LocalVariables|SetCurrentTeleporterCandidate (Utilities|Array|Get(acopy) _result (Math|Integer|%25%28Integer%29 (+ _index_2 (Variables|Default|GetCurrentTimeSelectedTargetIndex)) (Utilities|Array|Length _result)))))
                                  (return 0)
                                  (else
                                    (return (Variables|Default|GetCurrentTimeSelectedTarget))))))))))))
                (else
                  (return 0)))))
          (else
            _)))
      (:Is%20Not%20Valid
        _))
    (else
      _)))

(fn SetDestinationTeleporter
  (params (destination :type BP%20Creative%20Device%20Teleporter%20Object%20Reference))
  (bind _output_get (Variables|Default|SetDestinationTeleporter destination))
  (return _output_get))

(fn SetIsLinked
  (params (NewIsLinkedValue :type Boolean))
  (Variables|Default|SetbIsLinked NewIsLinkedValue))

(fn SetNewDestination
  (params (NewDestination :type BP%20Creative%20Device%20Teleporter%20Object%20Reference))
  (bind _self self)
  (if (Variables|Default|GetbLinkToTarget)
    (bind _newdestination (Variables|GetNewDestination)
      (:then
        (if (Variables|Default|GetbLinkToTarget _newdestination)
          (if (Variables|Default|GetbIsLinked _newdestination)
            (if (== (Variables|Default|GetDestinationTeleporter _newdestination) _self)
              (CallFunction|SetDestinationTeleporter _newdestination _self)
              (CallFunction|SetIsLinked true)
              (CallFunction|SetIsLinked _newdestination true)
              (else
                (CallFunction|SetIsLinked false)))
            (bind _destination_out (CallFunction|SetDestinationTeleporter _newdestination))
            (return _destination_out)
            (else
              _))
          (else
            _)))
      (:else
        _))
    (else
      _)))

(fn ShouldMaintainRelativeLocation
  (return (Math|Boolean|ANDBoolean (Variables|Knob|GetbMaintainRelativePosition) (> (Variables|Knob|GetKnob_EffectRadiusSize) 0.0) (not (Variables|Default|GetIgnoreEffectRadiusSettings)))))

(fn TeleportPlayerPawnTo
  (params (PawnToTeleport :type Fort%20Player%20Pawn%20Object%20Reference) (TargetTeleporter :type BP%20Creative%20Device%20Teleporter%20Object%20Reference) (bNewIgnoreEffectRadiusSettings :type Boolean))
  ;[1 1 1] Use the ability component system to trigger teleportation
  (Variables|Default|SetIgnoreEffectRadiusSettings bNewIgnoreEffectRadiusSettings)
  (Ability|SendGameplayEventtoActor self "(TagName=\"Abilities.Generic.ActiveAbility.Activate\")" (Utilities|Struct|MakeGameplayEventData "(TagName=\"Abilities.Generic.ActiveAbility.Activate\")" self PawnToTeleport TargetTeleporter)))

(fn TryToTeleportActorByOverlap
  (params (Actor :type Actor%20Object%20Reference%20%28by%20ref%29) (Forced%20Destination%20Teleport :type BP%20Creative%20Device%20Teleporter%20Object%20Reference))
  (bind _fortminigamelogic (Variables|Default|GetFortMinigameLogic))
  (bind _currentminigame (Variables|/Script/FortniteGame.FortMinigameLogicComponent|GetCurrentMinigame _fortminigamelogic))
  (bind _volume (Variables|/Script/FortniteGame.FortMinigame|GetVolume _currentminigame))
  (bind _charactermovement (Variables|/Script/Engine.Character|GetCharacterMovement (Variables|Default|GetPawnToTeleport)))
  (bind _pawntoteleport (Variables|Default|GetPawnToTeleport))
  (bind _asfort_player_pawn_athena (Utilities|Casting|CastToFortPlayerPawnAthena Actor))
  (bind _bcanteleportactor (CallFunction|CanTeleportActor _asfort_player_pawn_athena))
  (if _bcanteleportactor
    (Utilities|IsValid (Variables|GetForced%20Destination%20Teleport)
      (:Is%20Valid
        (bind _newdestinationteleporter (CallFunction|SetNewDestination (Variables|GetForced%20Destination%20Teleport)))
        (Utilities|IsValid _newdestinationteleporter
          (:Is%20Valid
            _)))
      (:Is%20Not%20Valid
        (bind _selectedteleporter (CallFunction|SelectDestinationTeleporter _asfort_player_pawn_athena))
        (bind _newdestinationteleporter_2 (CallFunction|SetNewDestination _selectedteleporter))
        (Utilities|IsValid _newdestinationteleporter_2
          (:Is%20Valid
            _))))
    (Networking|SwitchHasAuthority
      (:Authority
        (bind _returnvalue (Class|CreativeBlueprintLibrary|IsLocationRelevanttoActor _volume (Transformation|GetActorLocation (Variables|Default|GetDestinationTeleporter))))
        (if (and _returnvalue (Utilities|IsValid _volume :!class /Script/Engine.KismetSystemLibrary))
          (Variables|Default|SetPawnToTeleport _asfort_player_pawn_athena)
          (Utilities|FlowControl|Sequence
            (:then_0
              ;[1 1 1] VFX
              (Utilities|Time|ClearandInvalidateTimerbyHandle (Variables|VFX|GetTeleportEventResetTimer))
              (Variables|VFX|SetTeleporterEventType "Entered")
              (Variables|VFX|SetTeleporterEventFlash (Math|Byte|Byte+Byte (Variables|VFX|GetTeleporterEventFlash) 1))
              (bind _returnvalue_2 (Utilities|Time|SetTimerbyEvent (EventDispatchers|CreateEvent :!event ResetTeleportEventState) 0.1))
              (Variables|VFX|SetTeleportEventResetTimer _returnvalue_2))
            (:then_1
              ;[1 1 1] Momentum conservation
              (if (Variables|Knob|GetKnob_bConserveMomentum)
                (Variables|/Script/Engine.MovementComponent|SetVelocity (Math|Vector|RotateVector (Transformation|GetVelocity (Variables|Default|GetPawnToTeleport)) (Math|Rotator|Delta(Rotator) (Transformation|GetActorRotation (Variables|Default|GetDestinationTeleporter)) (Transformation|GetActorRotation))) (Pawn|GetMovementComponent (Variables|Default|GetPawnToTeleport)))
                (else
                  (Character|StopJumping _pawntoteleport)
                  (FortMovement|CancelJumpAbility _pawntoteleport false)
                  (Components|Movement|StopMovementImmediately _charactermovement)
                  (Pawn|Components|CharacterMovement|ClearAccumulatedForces _charactermovement))))
            (:then_2
              ;[1 1 1] Teleport The Player
              (CallFunction|TeleportPlayerPawnTo self (Variables|Default|GetPawnToTeleport) (Variables|Default|GetDestinationTeleporter) false)
              (GameplayEvent|BroadcastEvent (Variables|UserOptions|Events|GetOn%20Enter) 0 (Utilities|Struct|MakeGameplayEventPlayerEntered (Pawn|GetController (Variables|Default|GetPawnToTeleport))))
              (bind _result (Math|Integer|IncrementInt (Variables|Default|GetTeleporterUseCounter))
                (:%20%20
                  (Variables|Default|SetTeleporterUseCounter _result))))))))))

(fn UnregisterAllPlayers
  (Utilities|Set|Clear (Variables|Default|GetPawnsOnTeleporter))
  (Networking|SwitchHasAuthority
    (:Authority
      (bind (_arrayindex _arrayelement) (Utilities|Array|ReverseforEachLoop (Variables|Default|GetRegisteredPlayers))
        (:LoopBody
          (bind _asfort_player_pawn_athena (Utilities|Casting|CastToFortPlayerPawnAthena _arrayelement))
          (CallFunction|UnregisterPlayer _asfort_player_pawn_athena))
        (:Completed
          (Utilities|Array|Clear (Variables|Default|GetRegisteredPlayers)))))))

(fn UnregisterNullActors
  (bind (_arrayindex _arrayelement) (Utilities|Array|ReverseforEachLoop (Variables|Default|GetRegisteredPlayers))
    (:LoopBody
      (Utilities|IsValid _arrayelement
        (:Is%20Not%20Valid
          (Utilities|Array|RemoveIndex (Variables|Default|GetRegisteredPlayers) _arrayindex))))))

(fn UnregisterPlayer
  (params (FortPlayerPawnAthena :type Fort%20Player%20Pawn%20Athena%20Object%20Reference%20%28by%20ref%29))
  (bind _fortplayerpawnathena (Variables|GetFortPlayerPawnAthena))
  (Networking|SwitchHasAuthority
    (:Authority
      (Utilities|IsValid FortPlayerPawnAthena
        (:Is%20Valid
          (bind _returnvalue (Utilities|Array|RemoveItem (Variables|Default|GetRegisteredPlayers) FortPlayerPawnAthena))
          (if _returnvalue
            (CallFunction|BindorUnbindtoPlayerRegisterEvents _fortplayerpawnathena)
            (if (CallFunction|IsPortalDebuggingEnabled)
              (Development|PrintString (Utilities|String|Append "Unregistered Player: " (Utilities|GetDisplayName _fortplayerpawnathena) "In: " (Utilities|GetDisplayName self)) true true "(R=1.000000,G=0.720884,B=0.660858,A=1.000000)" 1e+01))))
        (:Is%20Not%20Valid
          (CallFunction|UnregisterNullActors))))))

(fn UpdateEffectRadiusScale
  (bind (_relativescale3d_x _relativescale3d_y _relativescale3d_z) (Math|Vector|BreakVector (Variables|/Script/Engine.SceneComponent|GetRelativeScale3D_X (Variables|Default|GetEffectRadius))))
  (Transformation|SetRelativeScale3D (Variables|Default|GetEffectRadius) (Math|Vector|MakeVector (Math|Float|Max(Float) (* (Variables|Knob|GetKnob_EffectRadiusSize) (Variables|Default|GetM%20To%20CMConversion)) 0.01) (Math|Float|Max(Float) (* (Variables|Knob|GetKnob_EffectRadiusSize) (Variables|Default|GetM%20To%20CMConversion)) 0.01) _relativescale3d_z))
  (Development|SetHiddeninGame (Variables|Default|GetEffectRadius) (Math|Float|NearlyEqual(Float) (Variables|Knob|GetKnob_EffectRadiusSize)) true))

(fn UpdateEnergyAttributes
  (bind _energycomponent (Variables|Default|GetEnergyComponent))
  (Energy|UpdateRegenRateAttr _energycomponent (Variables|Knob|GetKnob_RechargeRatePerSecond))
  (Energy|UpdateEnergyCostAttr _energycomponent (Variables|Knob|GetKnob_EnergyCostPerUse))
  (Energy|UpdateEnergyRechargeDelayAttr _energycomponent (Variables|Knob|GetKnob_DelayBeforeRecharge))
  (Energy|UpdateEnergyCapacityAttr _energycomponent (Variables|Knob|GetKnob_TeleporterEnergyCapacity)))

(fn ConstructionScript)
